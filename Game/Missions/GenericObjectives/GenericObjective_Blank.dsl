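; GenericObjective_Blank is a mission objective with no gameplay goal of its own, driven by mission events and hidden from the UI.
; StartObjective starts the objective playing, looks up the mission by GUID without using the result, and sets UI visibility to Invisible.
; StopObjective adds the event's gameplay tags to the mission as completion tags, stops the objective as Succeeded, and reapplies the invisible override.
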(event Objective|EventOnObjectiveLoaded)

(event MissionHelpers|HandleMissionEvent%28v2%29:StartObjective
  (params (InMissionGuid :type Guid%20Structure) (ObjectiveHandle :type Gameplay%20Tag%20Container%20Structure) (Params :type Fort%20Mission%20Event%20Params%20Object%20Reference) (DO_NOT_USE_THIS_OR_VARIABLES_BELOW :type Boolean) (EventFocus :type Object%20Reference) (EventContent :type Data%20Asset%20Object%20Reference) (EventInstigator :type Actor%20Object%20Reference) (GenericInt :type Integer) (GenericFloat :type Float%20%28single-precision%29) (GenericText :type Text%20%28by%20ref%29) (GameplayTags :type Gameplay%20Tag%20Container%20Structure) (MissionEvent :type Fort%20Mission%20Event%20Structure%20%28by%20ref%29))
  (Objective|StartPlayingObjective)
  (MissionManager|GetMissionfromGuid (Missions|GetMissionGuid))
  (Objective|SetUIVisibilityOverride "Invisible"))

(event MissionHelpers|HandleMissionEvent%28v2%29:StopObjective
  (params (InMissionGuid :type Guid%20Structure) (ObjectiveHandle :type Gameplay%20Tag%20Container%20Structure) (Params :type Fort%20Mission%20Event%20Params%20Object%20Reference) (DO_NOT_USE_THIS_OR_VARIABLES_BELOW :type Boolean) (EventFocus :type Object%20Reference) (EventContent :type Data%20Asset%20Object%20Reference) (EventInstigator :type Actor%20Object%20Reference) (GenericInt :type Integer) (GenericFloat :type Float%20%28single-precision%29) (GenericText :type Text%20%28by%20ref%29) (GameplayTags :type Gameplay%20Tag%20Container%20Structure) (MissionEvent :type Fort%20Mission%20Event%20Structure%20%28by%20ref%29))
  (MissionHelpers|AddCompletionTagstoMission (Missions|GetMissionGuid) GameplayTags)
  (Objective|StopObjective "Succeeded")
  (Objective|SetUIVisibilityOverride "Invisible"))

(fn ConstructionScript)
